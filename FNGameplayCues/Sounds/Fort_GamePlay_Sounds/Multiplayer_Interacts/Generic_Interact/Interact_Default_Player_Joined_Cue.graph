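# Sound cue for a player joining a generic multiplayer interaction, named as the default joined sound.
# A mixer layers one base wave, slightly delayed with randomized pitch and boosted volume, over a random pick from four equally weighted variant waves.
# It references shared player-interact attenuation and the consumables sound class, while overridden concurrency allows one instance per owner with a short retrigger time.
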SoundCue /FNGameplayCues/Sounds/Fort_GamePlay_Sounds/Multiplayer_Interacts/Generic_Interact/Interact_Default_Player_Joined_Cue
root=Mixer_0
nodes [
 {id=Delay_0, children [Modulator_0], properties {delayMax=0.1, delayMin=0.1}}
 {id=Mixer_0, children [Delay_0,Random_0], properties {inputVolume [1,1.2]}}
 {id=Modulator_0, children [WavePlayer_0], properties {pitchMax=1.03, pitchMin=0.94, volumeMax=1.6, volumeMin=1.4}}
 {id=Random_0, children [WavePlayer_1,WavePlayer_2,WavePlayer_3,WavePlayer_4], properties {weights [1,1,1,1]}}
 {
  id=WavePlayer_0
  properties {soundWaveAssetPtr=/FNGameplayCues/Sounds/Fort_GamePlay_Sounds/Multiplayer_Interacts/Generic_Interact/wav/Interact_Default_PlayerJoined}
 }
 {
  id=WavePlayer_1
  properties {soundWaveAssetPtr=/FNGameplayCues/Sounds/Fort_GamePlay_Sounds/Multiplayer_Interacts/Generic_Interact/wav/Interact_Default_PlayerJoined_01}
 }
 {
  id=WavePlayer_2
  properties {soundWaveAssetPtr=/FNGameplayCues/Sounds/Fort_GamePlay_Sounds/Multiplayer_Interacts/Generic_Interact/wav/Interact_Default_PlayerJoined_02}
 }
 {
  id=WavePlayer_3
  properties {soundWaveAssetPtr=/FNGameplayCues/Sounds/Fort_GamePlay_Sounds/Multiplayer_Interacts/Generic_Interact/wav/Interact_Default_PlayerJoined_03}
 }
 {
  id=WavePlayer_4
  properties {soundWaveAssetPtr=/FNGameplayCues/Sounds/Fort_GamePlay_Sounds/Multiplayer_Interacts/Generic_Interact/wav/Interact_Default_PlayerJoined_04}
 }
]
properties {
 attenuationSettings=&/Interact_Player_Att
 bOverrideConcurrency=true
 maxDistance=1300
 soundClassObject=&Fort_Consumables
 volumeMultiplier=1
 attenuationOverrides {falloffDistance=750, stereoSpread=750, attenuationShapeExtents {x=200}}
 concurrencyOverrides {bLimitToOwner=true, maxCount=1, retriggerTime=0.2}
}
